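# Control Rig for the Soccer Zombie backpack cosmetic, adding procedural wobble to three short rock bone chains on the pack.
# The forward solve runs three ChainHarmonics units in sequence on the right, middle and left dyn_rock chains.
# Each uses only Wave motion with its own speed and noise, with Pendulum and Reach disabled and debug draw left on.
# The graph leaves the hand, grass, small rock and card bones untouched.
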ControlRigBlueprint /BRCosmetics/Accessories/FORT_Backpacks/Backpack_F_MED_Soccer_Zombie/Meshes/F_MED_Soccer_Zombie_Pack_CtrlRig
graphs [
 RigVMFunctionLibrary {links [], nodes []}
 RigVMModel {nodes [
   n1 = &RigUnit_BeginExecution()
   /Script/ControlRig.RigUnit_ChainHarmonics(
    ChainRoot=dyn_rock_r_00
    DrawWorldOffset=(Rotation=(X=0,Y=0,Z=0,W=1),Translation=(X=0,Y=0,Z=0),Scale3D=(X=1,Y=1,Z=1))
    Pendulum=(bEnabled=False,PendulumStiffness=2,PendulumGravity=(X=0,Y=0,Z=0),PendulumBlend=0.75,PendulumDrag=0.98,PendulumMinimum=0,PendulumMaximum=0.1,PendulumEase=Linear,UnwindAxis=(X=0,Y=1,Z=0),UnwindMinimum=0.2,UnwindMaximum=0.05)
    Reach=(bEnabled=False,ReachTarget=(X=0,Y=0,Z=0),ReachAxis=(X=1,Y=0,Z=0),ReachMinimum=0,ReachMaximum=0,ReachEase=Linear)
    Speed=(X=0.1,Y=0.1,Z=0.1)
    Wave=(bEnabled=True,WaveFrequency=(X=0,Y=0,Z=0),WaveAmplitude=(X=0.5,Y=0.5,Z=0.5),WaveOffset=(X=1,Y=1,Z=2),WaveNoise=(X=4,Y=1,Z=1),WaveMinimum=1,WaveMaximum=1,WaveEase=SineEaseInOut)
    WaveCurve=(EditorCurveData=(Keys=((),(Time=1,Value=1)),PreInfinityExtrap=RCCE_Constant,PostInfinityExtrap=RCCE_Constant,DefaultValue=340282346638528859811704183484516925440))
    bDrawDebug=True
    ExecutePin=@n1.ExecutePin
   ) {is=struct RigVMUnitNode}
   > &RigUnit_ChainHarmonics(
    ChainRoot=dyn_rock_mid_00
    DrawWorldOffset=(Rotation=(X=0,Y=0,Z=0,W=1),Translation=(X=0,Y=0,Z=0),Scale3D=(X=1,Y=1,Z=1))
    Pendulum=(bEnabled=False,PendulumStiffness=2,PendulumGravity=(X=0,Y=0,Z=0),PendulumBlend=0.75,PendulumDrag=0.98,PendulumMinimum=0,PendulumMaximum=0.1,PendulumEase=Linear,UnwindAxis=(X=0,Y=1,Z=0),UnwindMinimum=0.2,UnwindMaximum=0.05)
    Reach=(bEnabled=False,ReachTarget=(X=0,Y=0,Z=0),ReachAxis=(X=1,Y=0,Z=0),ReachMinimum=0,ReachMaximum=0,ReachEase=Linear)
    Speed=(X=0.3,Y=0.3,Z=0.3)
    Wave=(bEnabled=True,WaveFrequency=(X=0,Y=0,Z=0),WaveAmplitude=(X=0.5,Y=0.5,Z=0.5),WaveOffset=(X=1,Y=1,Z=2),WaveNoise=(X=5,Y=1,Z=1),WaveMinimum=1,WaveMaximum=1,WaveEase=SineEaseInOut)
    WaveCurve=(EditorCurveData=(Keys=((),(Time=1,Value=1)),PreInfinityExtrap=RCCE_Constant,PostInfinityExtrap=RCCE_Constant,DefaultValue=340282346638528859811704183484516925440))
    bDrawDebug=True
   )
   > &RigUnit_ChainHarmonics(
    ChainRoot=dyn_rock_l_00
    DrawWorldOffset=(Rotation=(X=0,Y=0,Z=0,W=1),Translation=(X=0,Y=0,Z=0),Scale3D=(X=1,Y=1,Z=1))
    Pendulum=(bEnabled=False,PendulumStiffness=2,PendulumGravity=(X=0,Y=0,Z=0),PendulumBlend=0.75,PendulumDrag=0.98,PendulumMinimum=0,PendulumMaximum=0.1,PendulumEase=Linear,UnwindAxis=(X=0,Y=1,Z=0),UnwindMinimum=0.2,UnwindMaximum=0.05)
    Reach=(bEnabled=False,ReachTarget=(X=0,Y=0,Z=0),ReachAxis=(X=1,Y=0,Z=0),ReachMinimum=0,ReachMaximum=0,ReachEase=Linear)
    Speed=(X=0.2,Y=0.2,Z=0.2)
    Wave=(bEnabled=True,WaveFrequency=(X=0,Y=0,Z=0),WaveAmplitude=(X=0.3,Y=0.3,Z=0.3),WaveOffset=(X=1,Y=1,Z=2),WaveNoise=(X=3,Y=1,Z=1),WaveMinimum=1,WaveMaximum=1,WaveEase=SineEaseInOut)
    WaveCurve=(EditorCurveData=(Keys=((),(Time=1,Value=1)),PreInfinityExtrap=RCCE_Constant,PostInfinityExtrap=RCCE_Constant,DefaultValue=340282346638528859811704183484516925440))
    bDrawDebug=True
   )
  ]}
]
hierarchy [
 root {loc=<0 0 0>, rot=<0 0 0 1>}
 pack {parent=root, loc=<0 -0.7316 0>, rot=<0 0 0 1>}
 dyn_rock_base {parent=pack, loc=<0.9675 -7.8789 -7.6512>, rot=<0.6533 0.2706 -0.6533 0.2706>}
 wrist_r {parent=dyn_rock_base, loc=<-10.3543 -0.5431 -0.6158>, rot=<0 0 0 1>}
 hand_r {parent=wrist_r, loc=<-2.3784 0.2215 -0.1673>, rot=<-0.6735 -0.0361 0.0249 -0.7379>}
 index_metacarpal_r {parent=hand_r, loc=<-1.6534 -0.0181 2.9601>, rot=<0.0158 -0.0124 -0.0173 -0.9996>}
 index_01_r {parent=index_metacarpal_r, loc=<-8.3008 -0.0357 0.0524>, rot=<-0.0044 0.0485 0.0924 -0.9945>}
 index_02_r {parent=index_01_r, loc=<-3.3589 0.3003 -0.0453>, rot=<-0.002 0.0027 0.3225 -0.9466>}
 index_03_r {parent=index_02_r, loc=<-2.5562 0.1232 -0.0871>, rot=<-0.002 0.0067 0.2558 -0.9667>}
 middle_metacarpal_r {parent=hand_r, loc=<-2.1857 0.1726 0.5984>, rot=<0.006 0.0168 -0.0329 -0.9993>}
 middle_01_r {parent=middle_metacarpal_r, loc=<-7.3273 0.2447 -0.0129>, rot=<-0.0208 0.0042 0.0858 -0.9961>}
 middle_02_r {parent=middle_01_r, loc=<-4.0142 -0.1411 -0.07>, rot=<-0.0217 0.0259 0.3512 -0.9357>}
 middle_03_r {parent=middle_02_r, loc=<-2.6128 -0.0498 0.0267>, rot=<0.0083 0.0188 0.1708 -0.9851>}
 pinky_metacarpal_r {parent=hand_r, loc=<-1.7617 0.9753 -3.3814>, rot=<-0.0397 0.0814 -0.069 -0.9935>}
 pinky_01_r {parent=pinky_metacarpal_r, loc=<-7.1707 0.4832 0.1094>, rot=<-0.0654 -0.0035 0.2242 -0.9723>}
 pinky_02_r {parent=pinky_01_r, loc=<-2.8274 -0.0892 0.1043>, rot=<0.0037 0.0011 0.2854 -0.9584>}
 pinky_03_r {parent=pinky_02_r, loc=<-2.2022 -0.0313 -0.1112>, rot=<-0.0002 0.0098 0.1445 -0.9895>}
 ring_metacarpal_r {parent=hand_r, loc=<-2.2073 0.3627 -1.3638>, rot=<-0.0284 0.0722 -0.0458 -0.9959>}
 ring_01_r {parent=ring_metacarpal_r, loc=<-7.0297 0.6197 0.1446>, rot=<-0.0489 -0.0252 0.1892 -0.9804>}
 ring_02_r {parent=ring_01_r, loc=<-3.6171 -0.4648 0.1027>, rot=<0 0 0.2806 -0.9598>}
 ring_03_r {parent=ring_02_r, loc=<-2.4153 -0.1141 -0.1269>, rot=<0 0 0.1189 -0.9929>}
 thumb_01_r {parent=hand_r, loc=<-1.3119 0.3192 2.459>, rot=<0.6546 -0.4042 0.0664 -0.6354>}
 thumb_02_r {parent=thumb_01_r, loc=<-5.9336 0.0447 -0.0401>, rot=<-0.0037 0.0213 0.298 -0.9543>}
 thumb_03_r {parent=thumb_02_r, loc=<-2.2223 -0.3682 0.0907>, rot=<0.0118 -0.013 0.0066 -0.9998>}
 dyn_rock_l_00 {parent=dyn_rock_base, loc=<7.5732 2.721 0.1662>, rot=<0.3827 0 0 -0.9239>}
 dyn_rock_l_01 {parent=dyn_rock_l_00, loc=<3.1544 0 0>, rot=<0 0 0 1>}
 dyn_rock_l_02 {parent=dyn_rock_l_01, loc=<2.6643 0 0>, rot=<0 0 0 1>}
 dyn_rock_mid_00 {parent=dyn_rock_base, loc=<9.4347 -0.1006 0.751>, rot=<0.3827 0 0 -0.9239>}
 dyn_rock_mid_01 {parent=dyn_rock_mid_00, loc=<3.6547 0 0>, rot=<0 0 0 1>}
 dyn_rock_mid_02 {parent=dyn_rock_mid_01, loc=<2.786 0 0>, rot=<0 0 0 1>}
 dyn_rock_r_00 {parent=dyn_rock_base, loc=<8.8204 -4.1452 0.9245>, rot=<0.3827 0 0 -0.9239>}
 dyn_rock_r_01 {parent=dyn_rock_r_00, loc=<2.3756 0 0>, rot=<0 0 0 1>}
 dyn_rock_r_02 {parent=dyn_rock_r_01, loc=<2.0903 0 0>, rot=<0 0 0 1>}
 dyn_grass_01 {parent=dyn_rock_base, loc=<-0.4927 -10.0688 9.2793>, rot=<-0.2585 0.6812 0.2821 -0.6242>}
 dyn_grass_02 {parent=dyn_rock_base, loc=<-0.4404 -8.4377 11.1174>, rot=<0.0999 -0.866 -0.192 0.4508>}
 dyn_grass_03 {parent=dyn_rock_base, loc=<-0.0625 -6.1767 12.3983>, rot=<-0.0839 0.7595 0.1 -0.6373>}
 dyn_grass_04 {parent=dyn_rock_base, loc=<-0.0036 0.4927 11.9023>, rot=<0.056 0.7631 -0.0668 -0.6403>}
 dyn_grass_05 {parent=dyn_rock_base, loc=<-0.3709 3.493 11.4034>, rot=<0.056 0.7631 -0.0668 -0.6403>}
 dyn_grass_06 {parent=dyn_rock_base, loc=<0.0314 5.6061 9.3967>, rot=<0.1391 0.7479 -0.1658 -0.6276>}
 dyn_grass_07 {parent=dyn_rock_base, loc=<-0.181 7.6915 6.9911>, rot=<-0.4343 -0.5648 0.5175 0.4739>}
 dyn_grass_08 {parent=dyn_rock_base, loc=<0.0554 11.3281 2.8481>, rot=<0.5792 0.4056 -0.5792 -0.4056>}
 dyn_grass_09 {parent=dyn_rock_base, loc=<-0.3697 12.3471 -0.6071>, rot=<-0.4229 -0.5534 0.6039 0.3875>}
 dyn_grass_10 {parent=dyn_rock_base, loc=<-0.2405 12.7469 -4.0826>, rot=<-0.51 -0.4663 0.6077 0.3913>}
 dyn_grass_11 {parent=dyn_rock_base, loc=<-0.3788 8.793 -9.7712>, rot=<-0.5702 -0.3537 0.6795 0.2968>}
 dyn_grass_12 {parent=dyn_rock_base, loc=<-0.4068 6.3945 -11.2534>, rot=<-0.5687 -0.1069 0.8121 0.0749>}
 dyn_small_rock_01 {parent=dyn_rock_base, loc=<0.1208 7.654 -5.3352>, rot=<0 0.6088 0.7934 0>}
 dyn_small_rock_02 {parent=dyn_rock_base, loc=<-0.0417 0.3157 8.6443>, rot=<0 0.0436 0.999 0>}
 dyn_small_rock_03 {parent=dyn_rock_base, loc=<-0.1253 -8.9062 3.0116>, rot=<0 -0.3827 0.9239 0>}
 dyn_card {parent=pack, loc=<-10.2527 -12.4789 16.4559>, rot=<-0.5 -0.5 0.5 -0.5>}
]
