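# Physics asset for the grass on the zombie soccer backpack cosmetic, setting up simulation bodies for its dynamic bones.
# Twelve grass bones each get a box body offset along local X, and a small box sits on the rock base bone.
# Every grass body constrains directly to the rock base in a star layout rather than a chain.
# All joints share the same narrow swing and twist limits, allowing only slight sway.
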PhysicsAsset /BRCosmetics/Accessories/FORT_Backpacks/Backpack_F_MED_Soccer_Zombie/Meshes/F_MED_Soccer_Zombie_Pack_Grass_Physics
bodies {
 dyn_grass_01 [{shapeType=Box, extentX=3.76, extentY=2.76, extentZ=0.8125, center=<3 0 0>}]
 dyn_grass_02 [{shapeType=Box, extentX=4.2944, extentY=2.25, extentZ=0.8125, center=<3 0 0>}]
 dyn_grass_03 [{shapeType=Box, extentX=4.5743, extentY=2.2805, extentZ=0.8125, center=<5 0 0>}]
 dyn_grass_04 [{shapeType=Box, extentX=4.7521, extentY=2.3638, extentZ=0.8125, center=<5 0 0>}]
 dyn_grass_05 [{shapeType=Box, extentX=4.7419, extentY=2.25, extentZ=0.8125, center=<3 0 0>}]
 dyn_grass_06 [{shapeType=Box, extentX=4.1817, extentY=2.2944, extentZ=0.8125, center=<3 0 0>}]
 dyn_grass_07 [{shapeType=Box, extentX=5.1403, extentY=2.3589, extentZ=0.8622, center=<4 0 0>}]
 dyn_grass_08 [{shapeType=Box, extentX=3.76, extentY=2.76, extentZ=0.8125, center=<6 0 0>}]
 dyn_grass_09 [{shapeType=Box, extentX=4.6021, extentY=2.25, extentZ=0.8125, center=<4 0 0>}]
 dyn_grass_10 [{shapeType=Box, extentX=5.829, extentY=2.5087, extentZ=0.8125, center=<3 0 0>}]
 dyn_grass_11 [{shapeType=Box, extentX=3.76, extentY=2.76, extentZ=0.8125, center=<2 0 0>}]
 dyn_grass_12 [{shapeType=Box, extentX=4.4034, extentY=2.6447, extentZ=0.8125, center=<4 0 0>}]
 dyn_rock_base [{shapeType=Box, extentX=1.8125, extentY=2.75, extentZ=0.625}]
}
constraints [
 {bone1Name=dyn_grass_01, bone2Name=dyn_rock_base, swing1LimitDegrees=10, swing2LimitDegrees=20, twistLimitDegrees=3}
 {bone1Name=dyn_grass_02, bone2Name=dyn_rock_base, swing1LimitDegrees=10, swing2LimitDegrees=20, twistLimitDegrees=3}
 {bone1Name=dyn_grass_03, bone2Name=dyn_rock_base, swing1LimitDegrees=10, swing2LimitDegrees=20, twistLimitDegrees=3}
 {bone1Name=dyn_grass_05, bone2Name=dyn_rock_base, swing1LimitDegrees=10, swing2LimitDegrees=20, twistLimitDegrees=3}
 {bone1Name=dyn_grass_04, bone2Name=dyn_rock_base, swing1LimitDegrees=10, swing2LimitDegrees=20, twistLimitDegrees=3}
 {bone1Name=dyn_grass_06, bone2Name=dyn_rock_base, swing1LimitDegrees=10, swing2LimitDegrees=20, twistLimitDegrees=3}
 {bone1Name=dyn_grass_07, bone2Name=dyn_rock_base, swing1LimitDegrees=10, swing2LimitDegrees=20, twistLimitDegrees=3}
 {bone1Name=dyn_grass_08, bone2Name=dyn_rock_base, swing1LimitDegrees=10, swing2LimitDegrees=20, twistLimitDegrees=3}
 {bone1Name=dyn_grass_09, bone2Name=dyn_rock_base, swing1LimitDegrees=10, swing2LimitDegrees=20, twistLimitDegrees=3}
 {bone1Name=dyn_grass_10, bone2Name=dyn_rock_base, swing1LimitDegrees=10, swing2LimitDegrees=20, twistLimitDegrees=3}
 {bone1Name=dyn_grass_11, bone2Name=dyn_rock_base, swing1LimitDegrees=10, swing2LimitDegrees=20, twistLimitDegrees=3}
 {bone1Name=dyn_grass_12, bone2Name=dyn_rock_base, swing1LimitDegrees=10, swing2LimitDegrees=20, twistLimitDegrees=3}
]
