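# Physics asset for the Soccer Zombie backpack cosmetic mesh, giving it a minimal two-body setup.
# Small zero-length capsules, effectively spheres, sit on the dyn_rock_base and wrist_r bones, with the wrist shape slightly offset.
# A single constraint joins wrist_r to dyn_rock_base, with moderate, equal swing limit angles and a much tighter twist limit angle.
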PhysicsAsset /BRCosmetics/Accessories/FORT_Backpacks/Backpack_F_MED_Soccer_Zombie/Meshes/F_MED_Soccer_Zombie_Pack_Physics
bodies {dyn_rock_base [{shapeType=Capsule, length=0, radius=2}], wrist_r [{shapeType=Capsule, length=0, radius=2, center=<-0.25 0 0>}]}
constraints [{bone1Name=wrist_r, bone2Name=dyn_rock_base, swing1LimitDegrees=30, swing2LimitDegrees=30, twistLimitDegrees=5}]
