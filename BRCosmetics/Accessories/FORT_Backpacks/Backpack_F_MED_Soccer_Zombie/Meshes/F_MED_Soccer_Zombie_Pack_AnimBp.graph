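# Animation Blueprint for the Soccer Zombie backpack cosmetic, adding physics motion and floating rock bones.
# The event graph builds three vectors from sine and cosine of game time, which additively offset the rock bones.
# Two chained RigidBody sims, a low-gravity grass one and a normal-gravity one, precede those offsets and a Control Rig.
# A pose is layered on hand_r and dyn_card, then a weighted multi-bone constraint and offset place the card.
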AnimBlueprint /BRCosmetics/Accessories/FORT_Backpacks/Backpack_F_MED_Soccer_Zombie/Meshes/F_MED_Soccer_Zombie_Pack_AnimBp
target_skeleton=&/F_MED_Soccer_Zombie_Pack_Skeleton
graphs [
 AnimGraph {nodes [
   n1 = RigidBody(ComponentPose=Local To Component(LocalPose=Input Pose())) {
    bClampLinearTranslationLimitToRefPose=true
    bOverrideWorldGravity=true
    overridePhysicsAsset=&/F_MED_Soccer_Zombie_Pack_Grass_Physics
    simulationSpace=BaseBoneSpace
    alphaBoolBlend {blendOption=Linear}
    baseBoneRef {boneName=root}
    componentAppliedLinearAccClamp=<6000 6000 6000>
    componentLinearAccScale=<0.5 0.5 0.5>
    componentLinearVelScale=<0.5 0.5 0.5>
    overrideWorldGravity {z=-50}
   }
   n2 = RigidBody(ComponentPose=@n1) {
    bClampLinearTranslationLimitToRefPose=true
    bOverrideWorldGravity=true
    overridePhysicsAsset=&/F_MED_Soccer_Zombie_Pack_Physics
    alphaBoolBlend {blendOption=Linear}
    baseBoneRef {boneName=root}
    componentAppliedLinearAccClamp=<6000 6000 6000>
    componentLinearAccScale=<1 1 1>
    componentLinearVelScale=<1 1 1>
    overrideWorldGravity {z=-980}
   }
   n3 = "Transform (Modify) Bone - Bone: dyn_rock_r_02"(ComponentPose=@n2, Alpha=1, Translation=Get ShardFloatB() -> (ShardFloatB:Vector), Rotation="0,0,0", Scale="1,1,1") {translationMode=BMM_Additive, translationSpace=BCS_BoneSpace, alphaBoolBlend {blendOption=Linear}, boneToModify {boneName=dyn_rock_r_02}}
   n4 = "Transform (Modify) Bone - Bone: dyn_rock_mid_02"(ComponentPose=@n3, Alpha=1, Translation=Get ShardFloatC() -> (ShardFloatC:Vector), Rotation="0,0,0", Scale="1,1,1") {translationMode=BMM_Additive, translationSpace=BCS_BoneSpace, alphaBoolBlend {blendOption=Linear}, boneToModify {boneName=dyn_rock_mid_02}}
   n5 = "Transform (Modify) Bone - Bone: dyn_rock_l_02"(ComponentPose=@n4, Alpha=1, Translation=Get ShardFloatA() -> (ShardFloatA:Vector), Rotation="0,0,0", Scale="1,1,1") {translationMode=BMM_Additive, translationSpace=BCS_BoneSpace, alphaBoolBlend {blendOption=Linear}, boneToModify {boneName=dyn_rock_l_02}}
   n6 = Control Rig(Source=Component To Local(ComponentPose=@n5)) {
    alphaBoolBlend {blendOption=Linear}
    controlRigAssetReference {blueprintRigClass=&/F_MED_Soccer_Zombie_Pack_CtrlRig.F_MED_Soccer_Zombie_Pack_CtrlRig_C}
   }
   n7 = Sequence Player 'Soccer_Zombie_Pose'() {sequence=&/Soccer_Zombie_Pose}
   n8 = "Transform (Modify) Bone - Bone: dyn_rock_base"(
    ComponentPose=Local To Component(LocalPose=Layered blend per bone(BasePose=@n6, BlendPoses_0=@n7) {layerSetup [{branchFilters [{boneName=hand_r}, {boneName=dyn_card}]}]})
    Alpha=1
    Translation=vector * float(A=Get ShardFloatC(), B=-1)
    Rotation="0,0,-45"
    Scale="1,1,1"
   ) {
    rotationMode=BMM_Additive
    rotationSpace=BCS_BoneSpace
    translationMode=BMM_Additive
    translationSpace=BCS_BoneSpace
    alphaBoolBlend {blendOption=Linear}
    boneToModify {boneName=dyn_rock_base}
    rotation {roll=-45}
   }
   n9 = "Constraint - dyn_card (PPPPPPPPPPPPPPP)"!(
    ComponentPose=@n8
    Alpha=1
    ConstraintWeights_0=0.0002
    ConstraintWeights_1:float=0
    ConstraintWeights_2:float=0.1867
    ConstraintWeights_3:float=0
    ConstraintWeights_4:float=0
    ConstraintWeights_5:float=0
    ConstraintWeights_6:float=0.0055
    ConstraintWeights_7:float=0.7835
    ConstraintWeights_8:float=0
    ConstraintWeights_9:float=0
    ConstraintWeights_10:float=0
    ConstraintWeights_11:float=0.0232
    ConstraintWeights_12:float=0.0003
    ConstraintWeights_13:float=0.0004
    ConstraintWeights_14:float=0
   ) -> (Pose) {
    alphaBoolBlend {blendOption=Linear}
    boneToModify {boneName=dyn_card}
    constraintSetup [
     {targetBone {boneName=ring_metacarpal_r}}
     {targetBone {boneName=middle_01_r}}
     {targetBone {boneName=index_metacarpal_r}}
     {targetBone {boneName=ring_01_r}}
     {targetBone {boneName=middle_02_r}}
     {targetBone {boneName=ring_02_r}}
     {targetBone {boneName=hand_r}}
     {targetBone {boneName=thumb_01_r}}
     {targetBone {boneName=wrist_r}}
     {targetBone {boneName=pinky_metacarpal_r}}
     {targetBone {boneName=index_02_r}}
     {targetBone {boneName=middle_metacarpal_r}}
     {targetBone {boneName=thumb_02_r}}
     {targetBone {boneName=index_01_r}}
     {targetBone {boneName=pack}}
    ]
    constraintWeights [1,1,1,1,1,1,1,1,1,1,1,1,1,1,1]
   }
   n10 = "Transform (Modify) Bone - Bone: dyn_card"(ComponentPose=@n9, Alpha=1, Translation="3.442,3.426,-2.704", Rotation="4.741,23.899,49.065", Scale="1,1,1") {
    rotationMode=BMM_Additive
    rotationSpace=BCS_BoneSpace
    translationMode=BMM_Additive
    translationSpace=BCS_BoneSpace
    alphaBoolBlend {blendOption=Linear}
    boneToModify {boneName=dyn_card}
    rotation=<rot 4.741 23.8989 49.065>
    translation=<3.442 3.426 -2.704>
   }
   Output Pose(Result=Component To Local(ComponentPose=@n10)) {layerGroup=None}
  ]}
 EventGraph {nodes [
   n1 = Event BlueprintUpdateAnimation()
   n2 = GetGameTimeInSeconds()
   n3 = "Sin (Radians)"(A=float * float(A=float * float(A=1, B=1), B=@n2))
   n4 = "Sin (Radians)"(A=float * float(A=float * float(A=1, B=2), B=@n2))
   n5 = "Sin (Radians)"(A=float * float(A=float * float(A=1, B=1.5), B=@n2))
   Set ShardFloatA(
    execute=[@n1.then, @n12.then_0]
    ShardFloatA:Vector=MakeVector(X=float * float(A=@n5, B=0.3), Y=float * float(A=@n3, B=0.56), Z=float * float(A=@n4, B=0.22))
   ) -> (Output_Get:Vector)
   n6 = "Cos (Radians)"!(A:double=float * float(A=float * float(A=1, B=1.5), B=@n2)) -> (ReturnValue:double)
   n7 = "Cos (Radians)"(A=float * float(A=float * float(A=1, B=1.78), B=@n2))
   n8 = "Cos (Radians)"(A=float * float(A=float * float(A=1, B=1.15), B=@n2))
   > Set ShardFloatB(ShardFloatB:Vector=MakeVector(X=float * float(A=@n6, B=0.6), Y=float * float(A=@n7, B=0.725), Z=float * float(A=@n8, B=0.3))) -> (Output_Get:Vector)
   n9 = "Sin (Radians)"(A=float * float(A=float * float(A=1, B=1.5), B=@n2))
   n10 = "Sin (Radians)"(A=float * float(A=float * float(A=1, B=2), B=@n2))
   n11 = "Sin (Radians)"(A=float * float(A=float * float(A=1, B=1), B=@n2))
   > Set ShardFloatC(ShardFloatC:Vector=MakeVector(X=float * float(A=@n9, B=0.3), Y=float * float(A=@n11, B=0.6), Z=float * float(A=@n10, B=0.25))) -> (Output_Get:Vector)
   n12 = Sequence()
  ]}
]
variables [
 ShardFloatA {type=Vector}
 ShardFloatB {type=Vector}
 ShardFloatC {type=Vector}
]
